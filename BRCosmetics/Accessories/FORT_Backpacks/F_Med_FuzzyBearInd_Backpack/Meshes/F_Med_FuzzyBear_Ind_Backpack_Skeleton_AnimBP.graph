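# Animation Blueprint for the FuzzyBear Ind backpack cosmetic, adding procedural secondary motion to its firework, rocket, bottle and spark bones.
# The input pose runs through a serial chain of spring-based Anim Dynamics nodes, one per simulated bone, with varied damping and alphas.
# After the firework-mid simulation, Copy Bone Delta nodes pass its motion to the rocket and tube bones with negative rotation multipliers to counter-rotate them.
# The event graph has no logic.
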AnimBlueprint /BRCosmetics/Accessories/FORT_Backpacks/F_Med_FuzzyBearInd_Backpack/Meshes/F_Med_FuzzyBear_Ind_Backpack_Skeleton_AnimBP
target_skeleton=&/F_Med_FuzzyBear_Ind_Backpack_Skeleton
graphs [
 AnimGraph {nodes [
   AnimGraphNode_SubInput_0 = Input Pose()
   n1 = Anim Dynamics - Bone: dyn_firework_mid(ComponentPose=Local To Component(LocalPose=@AnimGraphNode_SubInput_0), Alpha=1, ExternalForce="0,0,0") {
    angularDampingOverride=0.995
    angularSpringConstant=15000
    bAngularSpring=true
    bGravityOverrideInSimSpace=true
    bOverrideAngularDamping=true
    linearDampingOverride=0.995
    linearSpringConstant=500
    alphaBoolBlend {blendOption=Linear}
    boundBone {boneName=dyn_firework_mid}
    componentAppliedLinearAccClamp=<1000 1000 1000>
    physicsBodyDefinitions [{
      sphereCollisionRadius=0
      boundBone {boneName=dyn_firework_mid}
      boxExtents=<80 80 10>
      constraintSetup {angularLimitsMax=<5 5 45>, angularLimitsMin=<-5 -5 -45>, angularTarget=<0 0 0>}
      localJointOffset=<0 0 -5>
     }]
   }
   n2 = Copy Bone Delta - Source Bone: dyn_firework_mid - Target Bone: dyn_rocket_down!(
    ComponentPose:Component Space Pose Link Structure=@n1
    Alpha:float=1
    bCopyTranslation:bool=true
    bCopyRotation:bool=true
    bCopyScale:bool=False
    TranslationMultiplier:float=1
    RotationMultiplier:float=-0.2
   ) -> (Pose:Component Space Pose Link Structure) {
    is=title AnimGraphNode_CopyBoneDelta
    bCopyTranslation=true
    rotationMultiplier=-0.5
    alphaBoolBlend {blendOption=Linear}
    sourceBone {boneName=dyn_firework_mid}
    targetBone {boneName=dyn_rocket_down}
   }
   n3 = Anim Dynamics - Bone: dyn_bottle_01(
    ComponentPose=Copy Bone Delta - Source Bone: dyn_firework_mid - Target Bone: dyn_firework_tube!(
     ComponentPose:Component Space Pose Link Structure=@n2
     Alpha:float=0.33
     bCopyTranslation:bool=true
     bCopyRotation:bool=true
     bCopyScale:bool=False
     TranslationMultiplier:float=1
     RotationMultiplier:float=-1
    ) -> (Pose:Component Space Pose Link Structure) {
     is=title AnimGraphNode_CopyBoneDelta
     alphaBoolBlend {blendOption=Linear}
     sourceBone {boneName=dyn_firework_mid}
     targetBone {boneName=dyn_firework_tube}
    }
    Alpha=1
    ExternalForce="0,0,0"
   ) {
    angularDampingOverride=0.998
    angularSpringConstant=25000
    bAngularSpring=true
    bGravityOverrideInSimSpace=true
    bOverrideAngularDamping=true
    linearDampingOverride=0.995
    linearSpringConstant=500
    alphaBoolBlend {blendOption=Linear}
    boundBone {boneName=dyn_bottle_01}
    componentAppliedLinearAccClamp=<1000 1000 1000>
    physicsBodyDefinitions [{
      sphereCollisionRadius=0
      boundBone {boneName=dyn_bottle_01}
      boxExtents=<80 80 10>
      constraintSetup {angularLimitsMax=<5 5 45>, angularLimitsMin=<-5 -5 -45>, angularTarget=<0 0 0>}
      localJointOffset=<0 0 -10>
     }]
   }
   n4 = Anim Dynamics - Bone: dyn_bottle_02(ComponentPose=@n3, Alpha=0.33, ExternalForce="0,0,0") {
    angularDampingOverride=0.995
    angularSpringConstant=25000
    bAngularSpring=true
    bGravityOverrideInSimSpace=true
    bOverrideAngularDamping=true
    linearDampingOverride=0.995
    linearSpringConstant=500
    alphaBoolBlend {blendOption=Linear}
    boundBone {boneName=dyn_bottle_02}
    componentAppliedLinearAccClamp=<1000 1000 1000>
    physicsBodyDefinitions [{
      sphereCollisionRadius=0
      boundBone {boneName=dyn_bottle_02}
      boxExtents=<80 80 10>
      constraintSetup {angularLimitsMax=<5 5 45>, angularLimitsMin=<-5 -5 -45>, angularTarget=<0 0 0>}
      localJointOffset=<0 0 -10>
     }]
   }
   n5 = Anim Dynamics - Bone: dyn_bottle_03(ComponentPose=@n4, Alpha=0.66, ExternalForce="0,0,0") {
    angularDampingOverride=0.999
    angularSpringConstant=25000
    bAngularSpring=true
    bGravityOverrideInSimSpace=true
    bOverrideAngularDamping=true
    linearDampingOverride=0.995
    linearSpringConstant=500
    alphaBoolBlend {blendOption=Linear}
    boundBone {boneName=dyn_bottle_03}
    componentAppliedLinearAccClamp=<1000 1000 1000>
    physicsBodyDefinitions [{
      sphereCollisionRadius=0
      boundBone {boneName=dyn_bottle_03}
      boxExtents=<74 76 10>
      constraintSetup {angularLimitsMax=<5 5 45>, angularLimitsMin=<-5 -5 -45>, angularTarget=<0 0 0>}
      localJointOffset=<0 0 -8>
     }]
   }
   n6 = Anim Dynamics - Bone: dyn_bottle_04(ComponentPose=@n5, Alpha=0.5, ExternalForce="0,0,0") {
    angularDampingOverride=0.9
    angularSpringConstant=25000
    bAngularSpring=true
    bGravityOverrideInSimSpace=true
    bOverrideAngularDamping=true
    linearDampingOverride=0.995
    linearSpringConstant=500
    alphaBoolBlend {blendOption=Linear}
    boundBone {boneName=dyn_bottle_04}
    componentAppliedLinearAccClamp=<1000 1000 1000>
    physicsBodyDefinitions [{
      sphereCollisionRadius=0
      boundBone {boneName=dyn_bottle_04}
      boxExtents=<80 80 10>
      constraintSetup {angularLimitsMax=<5 5 45>, angularLimitsMin=<-5 -5 -45>, angularTarget=<0 0 0>}
      localJointOffset=<0 0 -10>
     }]
   }
   n7 = Anim Dynamics - Bone: dyn_spark_01(ComponentPose=@n6, Alpha=1, ExternalForce="0,0,0") {
    angularDampingOverride=0.998
    angularSpringConstant=25000
    bAngularSpring=true
    bGravityOverrideInSimSpace=true
    bOverrideAngularDamping=true
    linearDampingOverride=0.995
    linearSpringConstant=500
    alphaBoolBlend {blendOption=Linear}
    boundBone {boneName=dyn_spark_01}
    componentAppliedLinearAccClamp=<1000 1000 1000>
    physicsBodyDefinitions [{
      sphereCollisionRadius=0
      boundBone {boneName=dyn_spark_01}
      boxExtents=<80 80 10>
      constraintSetup {angularLimitsMax=<5 5 45>, angularLimitsMin=<-5 -5 -45>, angularTarget=<0 0 0>}
      localJointOffset=<0 0 -10>
     }]
   }
   n8 = Anim Dynamics - Bone: dyn_spark_02(ComponentPose=@n7, Alpha=0.66, ExternalForce="0,0,0") {
    angularDampingOverride=0.999
    angularSpringConstant=25000
    bAngularSpring=true
    bGravityOverrideInSimSpace=true
    bOverrideAngularDamping=true
    linearDampingOverride=0.995
    linearSpringConstant=500
    alphaBoolBlend {blendOption=Linear}
    boundBone {boneName=dyn_spark_02}
    componentAppliedLinearAccClamp=<1000 1000 1000>
    physicsBodyDefinitions [{
      sphereCollisionRadius=0
      boundBone {boneName=dyn_spark_02}
      boxExtents=<74 76 10>
      constraintSetup {angularLimitsMax=<5 5 45>, angularLimitsMin=<-5 -5 -45>, angularTarget=<0 0 0>}
      localJointOffset=<0 0 -8>
     }]
   }
   n9 = Anim Dynamics - Bone: dyn_spark_03(ComponentPose=@n8, Alpha=0.33, ExternalForce="0,0,0") {
    angularDampingOverride=0.995
    angularSpringConstant=25000
    bAngularSpring=true
    bGravityOverrideInSimSpace=true
    bOverrideAngularDamping=true
    linearDampingOverride=0.995
    linearSpringConstant=500
    alphaBoolBlend {blendOption=Linear}
    boundBone {boneName=dyn_spark_03}
    componentAppliedLinearAccClamp=<1000 1000 1000>
    physicsBodyDefinitions [{
      sphereCollisionRadius=0
      boundBone {boneName=dyn_spark_03}
      boxExtents=<80 80 10>
      constraintSetup {angularLimitsMax=<5 5 45>, angularLimitsMin=<-5 -5 -45>, angularTarget=<0 0 0>}
      localJointOffset=<0 0 -10>
     }]
   }
   n10 = Anim Dynamics - Bone: dyn_spark_04(ComponentPose=@n9, Alpha=0.5, ExternalForce="0,0,0") {
    angularDampingOverride=0.9
    angularSpringConstant=25000
    bAngularSpring=true
    bGravityOverrideInSimSpace=true
    bOverrideAngularDamping=true
    linearDampingOverride=0.995
    linearSpringConstant=500
    alphaBoolBlend {blendOption=Linear}
    boundBone {boneName=dyn_spark_04}
    componentAppliedLinearAccClamp=<1000 1000 1000>
    physicsBodyDefinitions [{
      sphereCollisionRadius=0
      boundBone {boneName=dyn_spark_04}
      boxExtents=<80 80 10>
      constraintSetup {angularLimitsMax=<5 5 45>, angularLimitsMin=<-5 -5 -45>, angularTarget=<0 0 0>}
      localJointOffset=<0 0 -10>
     }]
   }
   n11 = Anim Dynamics - Bone: dyn_spark_05(ComponentPose=@n10, Alpha=0.5, ExternalForce="0,0,0") {
    angularDampingOverride=0.9
    angularSpringConstant=25000
    bAngularSpring=true
    bGravityOverrideInSimSpace=true
    bOverrideAngularDamping=true
    linearDampingOverride=0.995
    linearSpringConstant=500
    alphaBoolBlend {blendOption=Linear}
    boundBone {boneName=dyn_spark_05}
    componentAppliedLinearAccClamp=<1000 1000 1000>
    physicsBodyDefinitions [{
      sphereCollisionRadius=0
      boundBone {boneName=dyn_spark_05}
      boxExtents=<80 80 10>
      constraintSetup {angularLimitsMax=<5 5 45>, angularLimitsMin=<-5 -5 -45>, angularTarget=<0 0 0>}
      localJointOffset=<0 0 -10>
     }]
   }
   Output Pose(Result=Component To Local(ComponentPose=@n11)) {layerGroup=None}
  ]}
 EventGraph {nodes [Event BlueprintUpdateAnimation()]}
]
variables []
